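# Physics asset for the face accessory part of the ChaosLightning medium male player character, setting up physics for hair strands and an earring.
# Small capsule bodies on dyn bones hang from base_hair and base_earring anchors.
# Hair forms four two-bone chains plus the single dyn_hair03, while the earring is one jointed body.
# All joints lock twist and use narrow swing limits, except the earring's wider swing on one axis.
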PhysicsAsset /BRCosmetics/Characters/Player/Male/Medium/Bodies/M_MED_ChaosLightning/Meshes/Parts/M_MED_ChaosLightning_FaceAcc_Physics
bodies {
 base_earring [{shapeType=Capsule, length=0.505, radius=0.505, shapeName=base_earring_capsule, rotation=<rot 0 -90 0>}]
 base_hair [{shapeType=Capsule, length=0.505, radius=0.505, shapeName=base_hair_capsule, rotation=<rot 0 -90 0>}]
 dyn_earring [{
   shapeType=Capsule
   length=0.505
   radius=0.505
   shapeName=dyn_earring_capsule
   center=<1.6386 -0.0592 0.1113>
   rotation=<rot 0 88.8225 -88.0345>
  }]
 dyn_hair01 [{shapeType=Capsule, length=1, radius=1, shapeName=dyn_hair01_capsule, center=<2 0 0>, rotation=<rot 90 0 0>}]
 dyn_hair01_02 [{shapeType=Capsule, length=1, radius=1, shapeName=dyn_hair01_02_capsule, center=<2 0 0>, rotation=<rot 90 0 0>}]
 dyn_hair02 [{shapeType=Capsule, length=1, radius=1, shapeName=dyn_hair02_capsule, rotation=<rot 90 0 0>}]
 dyn_hair02_03 [{shapeType=Capsule, length=1, radius=1, shapeName=dyn_hair02_03_capsule, center=<1.4484 0 0>, rotation=<rot 90 0 0>}]
 dyn_hair03 [{shapeType=Capsule, length=1, radius=1, shapeName=dyn_hair03_capsule, center=<2.4575 0 0>, rotation=<rot 90 0 0>}]
 dyn_hair19 [{shapeType=Capsule, length=1, radius=1, shapeName=dyn_hair19_capsule, center=<1.1411 0 0>, rotation=<rot 90 0 0>}]
 dyn_hair19_02 [{shapeType=Capsule, length=1, radius=1, shapeName=dyn_hair19_02_capsule, center=<1.1411 0 0>, rotation=<rot 90 0 0>}]
 dyn_hair20 [{shapeType=Capsule, length=1, radius=1, shapeName=dyn_hair20_capsule, center=<1.1411 0 0>, rotation=<rot 90 0 0>}]
 dyn_hair20_02 [{shapeType=Capsule, length=1, radius=1, shapeName=dyn_hair20_02_capsule, center=<1.1411 0 0>, rotation=<rot 90 0 0>}]
}
constraints [
 {bone1Name=dyn_hair01, bone2Name=base_hair, swing1LimitDegrees=15, swing2LimitDegrees=10, twistLimitDegrees=1, twistMotion=Locked}
 {bone1Name=dyn_hair01_02, bone2Name=dyn_hair01, swing1LimitDegrees=8, swing2LimitDegrees=5, twistLimitDegrees=1, twistMotion=Locked}
 {bone1Name=dyn_hair02, bone2Name=base_hair, swing1LimitDegrees=5, swing2LimitDegrees=4, twistLimitDegrees=1, twistMotion=Locked}
 {bone1Name=dyn_hair02_03, bone2Name=dyn_hair02, swing1LimitDegrees=5, swing2LimitDegrees=12, twistLimitDegrees=1, twistMotion=Locked}
 {bone1Name=dyn_hair20_02, bone2Name=dyn_hair20, swing1LimitDegrees=8, swing2LimitDegrees=8, twistLimitDegrees=1, twistMotion=Locked}
 {bone1Name=dyn_hair19, bone2Name=base_hair, swing1LimitDegrees=13, swing2LimitDegrees=15, twistLimitDegrees=1, twistMotion=Locked}
 {bone1Name=dyn_hair19_02, bone2Name=dyn_hair19, swing1LimitDegrees=5, swing2LimitDegrees=10, twistLimitDegrees=1, twistMotion=Locked}
 {bone1Name=dyn_hair20, bone2Name=base_hair, swing1LimitDegrees=15, swing2LimitDegrees=5, twistLimitDegrees=1, twistMotion=Locked}
 {bone1Name=dyn_earring, bone2Name=base_earring, swing1LimitDegrees=15, swing2LimitDegrees=40, twistLimitDegrees=0, twistMotion=Locked}
 {bone1Name=dyn_hair03, bone2Name=base_hair, swing1LimitDegrees=5, swing2LimitDegrees=5, twistLimitDegrees=45, twistMotion=Locked}
]
